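# CP_Alpha_Mat is a UI-domain material for the color picker widget in the tech art Blutility tools, named for its alpha component.
# Its graph masks the red channel of the texture coordinates, giving a horizontal 0 to 1 gradient across the element.
# The graph shows no color parameters, textures or visible link from the mask to a material output.
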Material /Game/Blueprints/Blutilities/ForTechArt/Widgets/ColorPicker/Materials/CP_Alpha_Mat
graph {
 properties {materialDomain=MD_UI, pixelDepthOffsetMode=PDOM_Legacy, usedShadingModels=Default Lit}
 nodes [ComponentMask(None=TextureCoordinate()) {r=true}]
}
