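# UI-domain material for a color picker widget in a tech-art editor utility, producing a horizontal hue gradient.
# The U texture coordinate is raised to a Power parameter and scaled by a Mult parameter to form the hue shift percentage.
# A HueShift function applies that shift to the Color parameter, red by default, so the hue varies procedurally across the widget without a texture.
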Material /Game/Blueprints/Blutilities/ForTechArt/Widgets/ColorPicker/Materials/CP_Hue_Mat
graph {
 properties {materialDomain=MD_UI, pixelDepthOffsetMode=PDOM_Legacy, usedShadingModels=Default Lit}
 nodes [
  n1 = ComponentMask(None=TextureCoordinate()) {r=true}
  n2 = ScalarParameter() {defaultValue=1, parameterName=Power}
  MaterialFunctionCall(
   Hue Shift Percentage=Multiply(A=Power(Base=@n1, Exp=@n2) {constExponent=1.7}, B=ScalarParameter() {defaultValue=1, parameterName=Mult}) {constB=0.9}
   Texture=VectorParameter() {parameterName=Color, defaultValue=<rgba 1 0 0 1>}.RGB
  ) {materialFunction=&HueShift}
 ]
}
