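# UI material for the saturation gradient in a tech-art editor utility color picker widget.
# A lerp blends from the Color vector parameter, red by default, toward the unconnected B input's default white, using the horizontal UV coordinate as alpha.
# The result is a left-to-right ramp from the chosen color to white, with no texture sampling.
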Material /Game/Blueprints/Blutilities/ForTechArt/Widgets/ColorPicker/Materials/CP_Saturation_Mat
graph {
 properties {materialDomain=MD_UI, pixelDepthOffsetMode=PDOM_Legacy, usedShadingModels=Default Lit}
 nodes [LinearInterpolate(A=VectorParameter() {parameterName=Color, defaultValue=<rgba 1 0 0 1>}.RGB, Alpha=ComponentMask(None=TextureCoordinate()) {r=true})]
}
