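# UI-domain material, named as the slider thumb for a tech-art color picker Blutility widget.
# A Color vector parameter, red by default, is multiplied by the horizontal texture coordinate to form a black-to-color gradient.
# A separate If node turns the same coordinate into a hard black/white split at the midpoint, but does not feed the multiply.
# The graph shows no links from either branch to the material outputs.
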Material /Game/Blueprints/Blutilities/ForTechArt/Widgets/ColorPicker/Materials/CP_SliderThumb_Mat
graph {
 properties {materialDomain=MD_UI, pixelDepthOffsetMode=PDOM_Legacy, usedShadingModels=Default Lit}
 nodes [
  n1 = ComponentMask(None=TextureCoordinate()) {r=true}
  If(A=@n1, B=Constant() {r=0.5}, "A > B"=Constant() {r=1}, "A < B"=Constant())
  Multiply(A=VectorParameter() {parameterName=Color, defaultValue=<rgba 1 0 0 1>}.RGB, B=@n1)
 ]
}
