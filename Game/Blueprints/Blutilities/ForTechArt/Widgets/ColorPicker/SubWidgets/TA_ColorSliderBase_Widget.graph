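# Reusable UMG slider sub-widget for the tech-art color picker: a Slider with a thin material thumb and invisible bar over a Border background.
# PreConstruct sets the slider to startValue and, when valid, applies materialForSliderBG as the Border brush.
# SetColor pushes a color into the Border's dynamic material "Color" parameter.
# Slider changes fire the valueChanged dispatcher; setValue can optionally fire it too, and getValue reads the slider.
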WidgetBlueprint /Game/Blueprints/Blutilities/ForTechArt/Widgets/ColorPicker/SubWidgets/TA_ColorSliderBase_Widget
graphs [
 EventGraph {nodes [
   n1 = "On Value Changed (Slider_28)"!() -> (OutputDelegate, then, Value:float) {is=title K2Node_ComponentBoundEvent}
   > Call Value Changed!(execute:Exec, self:TA Color Slider Base Widget Object Reference, value:double=@n1.Value) -> (then:Exec) {is=title K2Node_CallDelegate}
   Event PreConstruct()
   > SetValue!(
    execute
    self:Slider Object Reference=Get Slider_28() -> (Slider_28:Slider Object Reference)
    InValue:float=Get startValue() -> (startValue:double)
   ) -> (then)
   > Is Valid(InputObject=Get materialForSliderBG() -> (materialForSliderBG:Material Interface Object Reference))
   > SetBrushFromMaterial(self:Border Object Reference=Get Border_0() -> (Border_0:Border Object Reference), Material=Get materialForSliderBG())
  ]}
 SetColor {nodes [
   n1 = SetColor!() -> (then, color:Linear Color Structure) {is=title K2Node_FunctionEntry}
   > n2 = GetDynamicMaterial(self:Border Object Reference=Get Border_0())
   > SetVectorParameterValue(self=@n2, ParameterName=Color, Value=@n1)
  ]}
 getValue {nodes [
   getValue() {is=title K2Node_FunctionEntry}
   > Return Node!(execute, value:double=GetValue!(self:Slider Object Reference=Get Slider_28()) -> (ReturnValue:float))
  ]}
 setValue {nodes [
   n1 = setValue!() -> (then, InValue:double=~, throwEvent:bool=~) {is=title K2Node_FunctionEntry}
   > SetValue(self=Get Slider_28(), InValue=@n1)
   > Branch(Condition=@n1.throwEvent)
   > Call Value Changed(value=@n1)
  ]}
 valueChanged {nodes [valueChanged!() -> (then, value:double=~) {is=title K2Node_FunctionEntry}]}
]
variables [
 materialForSliderBG {type=MaterialInterface, default=null}
 startValue {type=number}
 valueChanged {}
]
widgets {
 info {inheritedWidgetCount=0, parentClass=&UserWidget, rootWidgetClass=&CanvasPanel, widgetCount=3}
 widgets [
  CanvasPanel_0 {widgetClassPath=&CanvasPanel, widget=CanvasPanel_0}
  Border_0 {
   widgetClassPath=&Border
   bIsVariable=true
   parent=CanvasPanel_0
   widget=Border_0
   properties {verticalAlignment=VAlign_Top, padding=<ltrb 0 0 0 0>}
   slot {
    class=CanvasPanelSlot
    properties {bAutoSize=true, layoutData {alignment=<0 0>, anchors {maximum=<1 0>, minimum=<0 0>}, offsets=<ltrb 0 0 0 30>}}
   }
  }
  Slider_28 {
   widgetClassPath=/Script/UMG.Slider
   bIsVariable=true
   parent=Border_0
   widget=Slider_28
   properties {widgetStyle {
     barThickness=20
     disabledBarImage {
      bIsDynamicallyLoaded=false
      drawAs=Image
      imageType=NoImage
      mirroring=NoMirror
      resourceName=None
      resourceObject=/Engine/ArtTools/RenderToTexture/Textures/T_EV_BlankWhite_01
      tiling=NoTile
      imageSize=<16 16>
      margin=<ltrb 0 0 0 0>
      outlineSettings {
       bUseBrushTransparency=false
       roundingType=HalfHeightRadius
       width=0
       color {colorUseRule=UseColor_Specified, specifiedColor=<rgba 0 0 0 0>}
       cornerRadii=<0 0 0 0>
      }
      tintColor {colorUseRule=UseColor_Specified, specifiedColor=<rgba 0.7656 0.7656 0.7656 0>}
      uVRegion {bIsValid=false, max=<0 0>, min=<0 0>}
     }
     disabledThumbImage {
      bIsDynamicallyLoaded=false
      drawAs=Image
      imageType=FullColor
      mirroring=NoMirror
      resourceName=../../../Engine/Content/Slate/Common/Button_Disabled.png
      resourceObject=/Game/Blueprints/Blutilities/ForTechArt/Widgets/ColorPicker/Materials/CP_SliderThumb_Mat
      tiling=NoTile
      imageSize=<2 20>
      margin=<ltrb 0 0 0 0>
      outlineSettings {
       bUseBrushTransparency=false
       roundingType=HalfHeightRadius
       width=0
       color {colorUseRule=UseColor_Specified, specifiedColor=<rgba 0 0 0 0>}
       cornerRadii=<0 0 0 0>
      }
      tintColor {colorUseRule=UseColor_Specified, specifiedColor=<rgba 1 1 1 1>}
      uVRegion {bIsValid=false, max=<0 0>, min=<0 0>}
     }
     hoveredBarImage {
      bIsDynamicallyLoaded=false
      drawAs=Image
      imageType=NoImage
      mirroring=NoMirror
      resourceName=None
      resourceObject=&T_EV_BlankWhite_01
      tiling=NoTile
      imageSize=<16 16>
      margin=<ltrb 0 0 0 0>
      outlineSettings {
       bUseBrushTransparency=false
       roundingType=HalfHeightRadius
       width=0
       color {colorUseRule=UseColor_Specified, specifiedColor=<rgba 0 0 0 0>}
       cornerRadii=<0 0 0 0>
      }
      tintColor {colorUseRule=UseColor_Specified, specifiedColor=<rgba 0.7656 0.7656 0.7656 0>}
      uVRegion {bIsValid=false, max=<0 0>, min=<0 0>}
     }
     hoveredThumbImage {
      bIsDynamicallyLoaded=false
      drawAs=Image
      imageType=FullColor
      mirroring=NoMirror
      resourceName=../../../Engine/Content/Slate/Common/Button.png
      resourceObject=&CP_SliderThumb_Mat
      tiling=NoTile
      imageSize=<2 20>
      margin=<ltrb 0 0 0 0>
      outlineSettings {
       bUseBrushTransparency=false
       roundingType=HalfHeightRadius
       width=0
       color {colorUseRule=UseColor_Specified, specifiedColor=<rgba 0 0 0 0>}
       cornerRadii=<0 0 0 0>
      }
      tintColor {colorUseRule=UseColor_Specified, specifiedColor=<rgba 1 1 1 1>}
      uVRegion {bIsValid=false, max=<0 0>, min=<0 0>}
     }
     normalBarImage {
      bIsDynamicallyLoaded=false
      drawAs=Image
      imageType=NoImage
      mirroring=NoMirror
      resourceName=None
      resourceObject=&T_EV_BlankWhite_01
      tiling=NoTile
      imageSize=<16 16>
      margin=<ltrb 0 0 0 0>
      outlineSettings {
       bUseBrushTransparency=false
       roundingType=HalfHeightRadius
       width=0
       color {colorUseRule=UseColor_Specified, specifiedColor=<rgba 0 0 0 0>}
       cornerRadii=<0 0 0 0>
      }
      tintColor {colorUseRule=UseColor_Specified, specifiedColor=<rgba 0.7656 0.7656 0.7656 0>}
      uVRegion {bIsValid=false, max=<0 0>, min=<0 0>}
     }
     normalThumbImage {
      bIsDynamicallyLoaded=false
      drawAs=Image
      imageType=FullColor
      mirroring=NoMirror
      resourceName=../../../Engine/Content/Slate/Common/Button.png
      resourceObject=&CP_SliderThumb_Mat
      tiling=NoTile
      imageSize=<2 20>
      margin=<ltrb 0 0 0 0>
      outlineSettings {
       bUseBrushTransparency=false
       roundingType=HalfHeightRadius
       width=0
       color {colorUseRule=UseColor_Specified, specifiedColor=<rgba 0 0 0 0>}
       cornerRadii=<0 0 0 0>
      }
      tintColor {colorUseRule=UseColor_Specified, specifiedColor=<rgba 1 1 1 1>}
      uVRegion {bIsValid=false, max=<0 0>, min=<0 0>}
     }
    }}
   slot {class=BorderSlot, properties {verticalAlignment=VAlign_Top, padding=<ltrb 0 0 0 0>}}
  }
 ]
}
